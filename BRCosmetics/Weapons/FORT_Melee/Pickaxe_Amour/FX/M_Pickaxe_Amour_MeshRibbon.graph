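# Translucent unlit material for Niagara ribbon and mesh-particle trails on the Amour pickaxe.
# A scrolling tiled ribbon texture, offset by dynamic U offset and V-squeezed toward both ends by a taper, is tinted by particle color for emissive.
# Opacity combines texture alpha, particle alpha and smoothstep end fades, with depth fade skipped on mobile.
# A dynamic Length Control parameter wipes the trail along its length.
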Material /BRCosmetics/Weapons/FORT_Melee/Pickaxe_Amour/FX/M_Pickaxe_Amour_MeshRibbon
graph {
 properties {
  bUsedWithNiagaraMeshParticles=true
  bUsedWithNiagaraRibbons=true
  blendMode=BLEND_Translucent
  pixelDepthOffsetMode=PDOM_Legacy
  shadingModel=MSM_Unlit
  usedShadingModels=Unlit
 }
 nodes [
  n1 = ParticleColor()
  n2 = Constant() {r=1}
  n3 = ScalarParameter() {defaultValue=0.3, parameterName=Start Width Taper, sliderMax=1}
  n4 = ComponentMask(None=TextureCoordinate()) {r=true}
  n5 = ScalarParameter() {defaultValue=0.3, parameterName=End Width Taper, sliderMax=1}
  n6 = OneMinus(None=@n4)
  n7 = LinearInterpolate(
   A=ScalarParameter() {defaultValue=0.1, parameterName=Width Taper, sliderMax=1}
   B=Constant() {r=1}
   Alpha=Multiply(A=SmoothStep(Max=@n3, Value=@n4) {constMax=0.3}, B=SmoothStep(Max=@n5, Value=@n6) {constMax=0.3})
  )
  n8 = Multiply(
   A=MaterialFunctionCall(UVs=TextureCoordinate() {coordinateIndex=1}, Texture Scale=AppendVector(A=@n2, B=@n7)) {materialFunction=&ScaleUVsByCenter}.UVs
   B=AppendVector(
    A=ScalarParameter() {defaultValue=1, group=Tiling Tex, parameterName=Scale U}
    B=ScalarParameter() {defaultValue=1, group=Tiling Tex, parameterName=Scale V}
   )
  )
  n9 = Multiply(
   A=AppendVector(A=ScalarParameter() {defaultValue=0.5, group=Tiling Tex, parameterName=Pan U}, B=ScalarParameter() {group=Tiling Tex, parameterName=Pan V})
   B=Time() {period=2}
  )
  n10 = DynamicParameter() {paramNames [U offset, Length Control, Param3, Param4]}
  n11 = Constant()
  n12 = TextureSample(UVs=Add(A=Add(A=@n8, B=@n9), B=AppendVector(A=@n10.U offset, B=@n11))) {texture=/Game/Gadgets/Assets/VinderTech_GliderChute/Glider_TeddyBear/Effects/Textures/T_RibbonTrail_Tile}
  MaterialFunctionCall(Emissive=Multiply(A=@n1.RGB, B=@n12.RGB)) {materialFunction=&MF_FX_EmissiveAdjustments}
  n13 = ScalarParameter() {defaultValue=0.3, parameterName=Start Opacity Fade, sliderMax=1}
  n14 = ComponentMask(None=TextureCoordinate()) {r=true}
  n15 = ScalarParameter() {defaultValue=0.3, parameterName=End Opacity Fade, sliderMax=1}
  n16 = OneMinus(None=@n14)
  n17 = Multiply(
   A=Multiply(A=@n1.A, B=@n12.A)
   B=Multiply(A=SmoothStep(Max=@n13, Value=@n14) {constMax=0.3}, B=SmoothStep(Max=@n15, Value=@n16) {constMax=0.3})
  )
  n18 = DepthFade(Opacity=@n17, FadeDistance=ScalarParameter() {defaultValue=100, parameterName=Depth Fade Distance})
  n19 = ScalarParameter() {parameterName=Length Control Fade, sliderMax=1}
  n20 = OneMinus(None=ComponentMask(None=TextureCoordinate()) {r=true})
  n21 = DynamicParameter() {paramNames [U offset, Length Control, Param3, Param4]}
  Multiply(
   A=ShadingPathSwitch(Default=@n18, Deferred=@n18, Mobile=@n17) {desc=Migrated from FeatureLevelSwitch}
   B=SmoothStep(Min=Subtract(B=@n19), Value=Add(A=@n20, B=@n21.Length Control))
  )
  ScalarParameter() {parameterName=Length Control, sliderMax=1}
 ]
}
