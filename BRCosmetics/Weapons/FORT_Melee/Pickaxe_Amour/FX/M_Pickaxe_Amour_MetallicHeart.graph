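# Simple Default Lit material for metallic heart mesh particles in the Amour pickaxe effects.
# It is flagged for Niagara mesh particles and has ray traced shadow casting turned off.
# The graph holds only a Particle Color node and three scalar constants, with no textures or parameters; the file does not show which material inputs they feed.
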Material /BRCosmetics/Weapons/FORT_Melee/Pickaxe_Amour/FX/M_Pickaxe_Amour_MetallicHeart
graph {
 properties {bCastRayTracedShadows=false, bUsedWithNiagaraMeshParticles=true, pixelDepthOffsetMode=PDOM_Legacy, usedShadingModels=Default Lit}
 nodes [
  Constant() {r=1}
  Constant()
  Constant()
  ParticleColor()
 ]
}
